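# Physics asset for the Tide King backpack cosmetic, giving its pack and dangling strap bones simple collision and constrained motion.
# Capsules cover the root and pack bones, while three spheres follow the strap bones, growing larger along the chain.
# The straps chain from the pack through tight swing and tighter twist limits, while the pack-to-root constraint leaves swing and twist free.
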PhysicsAsset /BRCosmetics/Accessories/FORT_Backpacks/Backpack_M_MED_TideKing/Meshes/M_MED_TideKing_Pack_Physics_FE
bodies {
 dyn_strap_1 [{shapeType=Sphere, radius=2.5, shapeName=dyn_strap_1_sphere, center=<3.2695 0 0>}]
 dyn_strap_2 [{shapeType=Sphere, radius=3, shapeName=dyn_strap_2_sphere, center=<2.8393 0 0>}]
 dyn_strap_3 [{shapeType=Sphere, radius=5, shapeName=dyn_strap_3_sphere, center=<7.7938 0.7424 1.6154>}]
 pack [{shapeType=Capsule, length=22.5344, radius=5.6513, center=<-8.8589 -7.3117 -7.1284>, rotation=<rot -34.3021 -23.1729 1.5264>}]
 root [{shapeType=Capsule, length=45.5, radius=23.505, shapeName=root_capsule, center=<0 24.7365 -18.3803>, rotation=<rot 0 0 5>}]
}
constraints [
 {
  bone1Name=pack
  bone2Name=root
  swing1LimitDegrees=45
  swing1Motion=Free
  swing2LimitDegrees=45
  swing2Motion=Free
  twistLimitDegrees=45
  twistMotion=Free
 }
 {bone1Name=dyn_strap_1, bone2Name=pack, swing1LimitDegrees=20, swing2LimitDegrees=20, twistLimitDegrees=6}
 {bone1Name=dyn_strap_2, bone2Name=dyn_strap_1, swing1LimitDegrees=20, swing2LimitDegrees=20, twistLimitDegrees=6}
 {bone1Name=dyn_strap_3, bone2Name=dyn_strap_2, swing1LimitDegrees=20, swing2LimitDegrees=20, twistLimitDegrees=6}
]
