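# Anim Blueprint for the TideKing backpack cosmetic, adding secondary motion to its blob creature and bubbles.
# It offsets and scales the pack bone, nudges a strap, then runs chained pack-relative AnimDynamics on blob and bubble bones into cached pose 'preRB'.
# That pose feeds two RigidBody sims; bIsInFrontEnd selects the front-end physics asset, simulated in world space with damped component motion, or the regular one.
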AnimBlueprint /BRCosmetics/Accessories/FORT_Backpacks/Backpack_M_MED_TideKing/Meshes/M_MED_TideKing_Pack_AnimBP
target_skeleton=&/M_MED_TideKing_Pack_Skeleton
graphs [
 AnimGraph {nodes [
   n1 = RigidBody(ComponentPose=Local To Component(LocalPose=Use cached pose 'preRB'())) {
    overridePhysicsAsset=&/M_MED_TideKing_Pack_Physics
    alphaBoolBlend {blendOption=Linear}
    baseBoneRef {boneName=pack}
    componentAppliedLinearAccClamp=<5000 5000 5000>
    componentLinearAccScale=<0.7 0.7 0.7>
    componentLinearVelScale=<0.7 0.7 0.7>
   }
   n2 = RigidBody(ComponentPose=Local To Component(LocalPose=Use cached pose 'preRB'())) {
    bOverrideWorldGravity=true
    overridePhysicsAsset=&/M_MED_TideKing_Pack_Physics_FE
    simulationSpace=WorldSpace
    alphaBoolBlend {blendOption=Linear}
    baseBoneRef {boneName=pack}
    componentAppliedLinearAccClamp=<100 100 100>
    componentLinearAccScale=<0.1 0.1 0.1>
    componentLinearVelScale=<0.1 0.1 0.1>
    overrideWorldGravity {z=-980}
   }
   Output Pose(Result=Blend Poses by bool(BlendPose_0=Component To Local(ComponentPose=@n2), BlendPose_1=Component To Local(ComponentPose=@n1), bActiveValue=Get bIsInFrontEnd()) {blendType=Linear}) {layerGroup=None}
   n3 = "Transform (Modify) Bone - Bone: pack"(ComponentPose=Local To Component(LocalPose=Input Pose()), Translation="1,0,0", Rotation="0,-15,0", Scale="0.8,0.8,0.8") {
    rotationMode=BMM_Additive
    scaleMode=BMM_Replace
    scaleSpace=BCS_BoneSpace
    translationMode=BMM_Additive
    alphaBoolBlend {blendOption=Linear}
    boneToModify {boneName=pack}
    rotation {yaw=-15}
    scale=<0.8 0.8 0.8>
    translation {x=1}
   }
   n4 = "Transform (Modify) Bone - Bone: dyn_strap_1"(ComponentPose=@n3, Alpha=1, Translation="1,0,0", Rotation="0,0,0", Scale="1,1,1") {
    rotationMode=BMM_Additive
    scaleMode=BMM_Replace
    scaleSpace=BCS_BoneSpace
    translationMode=BMM_Additive
    alphaBoolBlend {blendOption=Linear}
    boneToModify {boneName=dyn_strap_1}
    translation {x=1}
   }
   n5 = Anim Dynamics - Bone: dyn_blobguy_head_1(ComponentPose=@n4, Alpha=0.5, ExternalForce="0,0,0") {
    angularSpringConstant=0.9
    bAngularSpring=true
    bUseGravityOverride=true
    bUsePlanarLimit=false
    linearSpringConstant=0.9
    simulationSpace=BoneRelative
    alphaBoolBlend {blendOption=Linear}
    boundBone {boneName=dyn_blobguy_head_1}
    componentAppliedLinearAccClamp=<100 100 100>
    componentLinearAccScale=<0.1 0.1 0.1>
    componentLinearVelScale=<0.1 0.1 0.1>
    gravityOverride {z=-980}
    physicsBodyDefinitions [{
      boundBone {boneName=dyn_blobguy_head_1}
      boxExtents=<10 10 10>
      constraintSetup {angularLimitsMax=<10 15 0>, angularLimitsMin=<-10 -15 -5>, angularTarget=<0 0 0>}
      localJointOffset=<-1 0 0>
     }]
    relativeSpaceBone {boneName=pack}
   }
   n6 = Anim Dynamics - Bone: dyn_blobguy_big_bubble_1(ComponentPose=@n5, Alpha=0.2, ExternalForce="0,0,0") {
    angularSpringConstant=0.9
    bAngularSpring=true
    bLinearSpring=true
    bUseGravityOverride=true
    linearSpringConstant=0.9
    simulationSpace=BoneRelative
    alphaBoolBlend {blendOption=Linear}
    boundBone {boneName=dyn_blobguy_big_bubble_1}
    componentAppliedLinearAccClamp=<100 100 100>
    componentLinearAccScale=<0.1 0.1 0.1>
    componentLinearVelScale=<0.1 0.1 0.1>
    physicsBodyDefinitions [{
      boundBone {boneName=dyn_blobguy_big_bubble_1}
      boxExtents=<10 10 10>
      constraintSetup {angularLimitsMax=<10 15 5>, angularLimitsMin=<-10 -15 -15>, angularTarget=<0 0 0>, linearAxesMax=<0.1 0.5 0.8>, linearAxesMin=<-2 -0.1 -0.1>}
      localJointOffset=<-1 0 0>
     }]
    planarLimits [{drivingBone {boneName=dyn_blobguy_head_1}, planeTransform {location=<2 0 0>, rotation=<rot 90 0 0>}}]
    relativeSpaceBone {boneName=pack}
   }
   n7 = Anim Dynamics - Bone: dyn_blobguy_small_bubble_1(ComponentPose=@n6, Alpha=0.3, ExternalForce="0,0,0") {
    angularSpringConstant=0.9
    bAngularSpring=true
    bLinearSpring=true
    bUseGravityOverride=true
    gravityScale=0
    linearSpringConstant=0.9
    simulationSpace=BoneRelative
    alphaBoolBlend {blendOption=Linear}
    boundBone {boneName=dyn_blobguy_small_bubble_1}
    componentAppliedLinearAccClamp=<100 100 100>
    componentLinearAccScale=<0.1 0.1 0.1>
    componentLinearVelScale=<0.1 0.1 0.1>
    gravityOverride {z=-20}
    physicsBodyDefinitions [{
      boundBone {boneName=dyn_blobguy_small_bubble_1}
      boxExtents=<10 10 10>
      constraintSetup {angularLimitsMax=<10 15 5>, angularLimitsMin=<-10 -15 -15>, angularTarget=<0 0 0>, linearAxesMax=<2 2 2>, linearAxesMin=<-0.1 -0.1 -0.5>}
      localJointOffset=<-1 0 0>
     }]
    planarLimits [{drivingBone {boneName=dyn_blobguy_head_1}, planeTransform {location=<1 0 0>, rotation=<rot 90 0 0>}}]
    relativeSpaceBone {boneName=pack}
   }
   n8 = Anim Dynamics - Bone: dyn_blobguy_fourth_1(ComponentPose=@n7, Alpha=1, ExternalForce="0,0,0") {
    bLinearSpring=true
    linearSpringConstant=0.9
    simulationSpace=BoneRelative
    alphaBoolBlend {blendOption=Linear}
    boundBone {boneName=dyn_blobguy_fourth_1}
    componentAppliedLinearAccClamp=<100 100 100>
    componentLinearAccScale=<0.1 0.1 0.1>
    componentLinearVelScale=<0.1 0.1 0.1>
    gravityOverride {z=-10}
    physicsBodyDefinitions [{
      boundBone {boneName=dyn_blobguy_fourth_1}
      boxExtents=<10 10 10>
      constraintSetup {angularLimitsMax=<10 15 3>, angularLimitsMin=<-10 -15 -15>, angularTarget=<0 0 0>}
      localJointOffset=<-1 0 0>
     }]
    relativeSpaceBone {boneName=pack}
   }
   n9 = Anim Dynamics - Bone: dyn_blobguy_first_1(ComponentPose=@n8, Alpha=1, ExternalForce="0,0,0") {
    bLinearSpring=true
    linearSpringConstant=0.9
    simulationSpace=BoneRelative
    alphaBoolBlend {blendOption=Linear}
    boundBone {boneName=dyn_blobguy_first_1}
    componentAppliedLinearAccClamp=<100 100 100>
    componentLinearAccScale=<0.1 0.1 0.1>
    componentLinearVelScale=<0.1 0.1 0.1>
    gravityOverride {z=-10}
    physicsBodyDefinitions [{
      boundBone {boneName=dyn_blobguy_first_1}
      boxExtents=<10 10 10>
      constraintSetup {angularLimitsMax=<4 10 10>, angularLimitsMin=<-4 -10 -2>, angularTarget=<0 0 0>}
      localJointOffset=<-2 0 0>
     }]
    relativeSpaceBone {boneName=pack}
   }
   n10 = Anim Dynamics - Bone: dyn_blobguy_first_2(ComponentPose=@n9, Alpha=0.2, ExternalForce="0,0,0") {
    bLinearSpring=true
    bUseGravityOverride=true
    linearSpringConstant=0.9
    simulationSpace=BoneRelative
    alphaBoolBlend {blendOption=Linear}
    boundBone {boneName=dyn_blobguy_first_2}
    componentAppliedLinearAccClamp=<100 100 100>
    componentLinearAccScale=<0.1 0.1 0.1>
    componentLinearVelScale=<0.1 0.1 0.1>
    gravityOverride {z=-10}
    physicsBodyDefinitions [{
      boundBone {boneName=dyn_blobguy_first_2}
      boxExtents=<10 10 10>
      constraintSetup {angularLimitsMax=<4 10 10>, angularLimitsMin=<-4 -10 -2>, angularTarget=<0 0 0>}
      localJointOffset=<-2 0 0>
     }]
    relativeSpaceBone {boneName=pack}
   }
   n11 = Anim Dynamics - Bone: dyn_blobguy_second_1(ComponentPose=@n10, Alpha=1, ExternalForce="0,0,0") {
    bLinearSpring=true
    linearSpringConstant=0.9
    simulationSpace=BoneRelative
    alphaBoolBlend {blendOption=Linear}
    boundBone {boneName=dyn_blobguy_second_1}
    componentAppliedLinearAccClamp=<100 100 100>
    componentLinearAccScale=<0.1 0.1 0.1>
    componentLinearVelScale=<0.1 0.1 0.1>
    gravityOverride {z=-10}
    physicsBodyDefinitions [{
      boundBone {boneName=dyn_blobguy_second_1}
      boxExtents=<10 10 10>
      constraintSetup {angularLimitsMax=<4 10 5>, angularLimitsMin=<-4 -10 -5>, angularTarget=<0 0 0>}
      localJointOffset=<-2 0 0>
     }]
    relativeSpaceBone {boneName=pack}
   }
   n12 = Anim Dynamics - Bone: dyn_blobguy_second_2(ComponentPose=@n11, Alpha=0.5, ExternalForce="0,0,0") {
    bLinearSpring=true
    bUseGravityOverride=true
    linearSpringConstant=0.9
    simulationSpace=BoneRelative
    alphaBoolBlend {blendOption=Linear}
    boundBone {boneName=dyn_blobguy_second_2}
    componentAppliedLinearAccClamp=<100 100 100>
    componentLinearAccScale=<0.1 0.1 0.1>
    componentLinearVelScale=<0.1 0.1 0.1>
    gravityOverride {z=-10}
    physicsBodyDefinitions [{
      boundBone {boneName=dyn_blobguy_second_2}
      boxExtents=<10 10 10>
      constraintSetup {angularLimitsMax=<4 10 5>, angularLimitsMin=<-4 -10 -5>, angularTarget=<0 0 0>}
      localJointOffset=<-2 0 0>
     }]
    relativeSpaceBone {boneName=pack}
   }
   n13 = Anim Dynamics - Bone: dyn_blobguy_third_1(ComponentPose=@n12, Alpha=1, ExternalForce="0,0,0") {
    bLinearSpring=true
    linearSpringConstant=0.9
    simulationSpace=BoneRelative
    alphaBoolBlend {blendOption=Linear}
    boundBone {boneName=dyn_blobguy_third_1}
    componentAppliedLinearAccClamp=<100 100 100>
    componentLinearAccScale=<0.1 0.1 0.1>
    componentLinearVelScale=<0.1 0.1 0.1>
    gravityOverride {z=-10}
    physicsBodyDefinitions [{
      boundBone {boneName=dyn_blobguy_third_1}
      boxExtents=<10 10 10>
      constraintSetup {angularLimitsMax=<4 10 5>, angularLimitsMin=<-4 -10 -5>, angularTarget=<0 0 0>}
      localJointOffset=<-2 0 0>
     }]
    relativeSpaceBone {boneName=pack}
   }
   n14 = Anim Dynamics - Bone: dyn_blobguy_third_2(ComponentPose=@n13, Alpha=0.3, ExternalForce="0,0,0") {
    bLinearSpring=true
    bUseGravityOverride=true
    linearSpringConstant=0.9
    simulationSpace=BoneRelative
    alphaBoolBlend {blendOption=Linear}
    boundBone {boneName=dyn_blobguy_third_2}
    componentAppliedLinearAccClamp=<100 100 100>
    componentLinearAccScale=<0.1 0.1 0.1>
    componentLinearVelScale=<0.1 0.1 0.1>
    gravityOverride {z=-10}
    physicsBodyDefinitions [{
      boundBone {boneName=dyn_blobguy_third_2}
      boxExtents=<10 10 10>
      constraintSetup {angularLimitsMax=<4 10 5>, angularLimitsMin=<-4 -10 -5>, angularTarget=<0 0 0>}
      localJointOffset=<-2 0 0>
     }]
    relativeSpaceBone {boneName=pack}
   }
   Save cached pose 'preRB'(Pose=Component To Local(ComponentPose=@n14)) {cacheName=preRB}
  ]}
 EventGraph {nodes [
   TryGetPawnOwner()
   Event BlueprintUpdateAnimation()
  ]}
]
variables []
